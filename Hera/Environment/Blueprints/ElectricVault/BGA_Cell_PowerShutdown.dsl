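; Invulnerable hackable button for the Electric Vault's cells that reports an inside-cell hack to the vault manager.
; It is a long-use interaction with a duration from a hotfixable scalable float; completing it disables the button and calls InsideCellHackInitiated.
; The server tracks interacting pawns to set HackingInProgress, whose RepNotify starts or fades looping audio on clients.
; SetPowerState toggles whether the button can be pressed.
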(event EventBeginPlay
  ; Init
  (Variables|BuildingActor|SetbIsInvulnerable true)
  (Variables|Default|SetHackDuration (Math|Conversions|ToFloat(Integer) (ScalableFloat|GetValueasInteger (Variables|Hotfixable|GetPrisonHackDuration)))))

(event Interaction|EventBlueprintOnInteract
  (params (InteractingPawn :type Fort%20Pawn%20Object%20Reference) (InteractionBeingAttempted :type EInteractionBeingAttempted%20Enum :default "FirstInteraction"))
  ; Button Interaction Logic
  (Variables|Default|SetButtonCanBePressed false)
  (Class|BGAElectricVaultManager|InsideCellHackInitiated (Variables|Default|GetVaultManagerRef))
  (Variables|Default|SetHackingInProgress false)
  (Utilities|Array|Clear (Variables|Default|GetInteractingPawnsOnServer)))

(event Interaction|EventOnServerNotifyStartLongUse
  (params (InteractingPawn :type Fort%20Player%20Pawn%20Object%20Reference))
  ; Button Interaction Logic
  (Utilities|Array|Add (Variables|Default|GetInteractingPawnsOnServer) InteractingPawn)
  (Variables|Default|SetHackingInProgress true))

(event Interaction|EventOnServerNotifyEndLongUse
  (params (InteractingPawn :type Fort%20Player%20Pawn%20Object%20Reference) (bUseCompleted :type Boolean))
  ; Button Interaction Logic
  (Utilities|Array|RemoveItem (Variables|Default|GetInteractingPawnsOnServer) InteractingPawn)
  (if (Utilities|Array|IsEmpty (Variables|Default|GetInteractingPawnsOnServer))
    (Variables|Default|SetHackingInProgress false)))

(event Custom|SetPowerState
  (params (PowerEnabled :type Boolean))
  ; Power State
  (if PowerEnabled
    (Variables|Default|SetButtonCanBePressed true)
    (else
      (Variables|Default|SetButtonCanBePressed false))))

(fn BlueprintCanInteract
  (params (InteractingPawn :type Fort%20Pawn%20Object%20Reference) (InteractionBeingAttempted :type EInteractionBeingAttempted%20Enum :default "FirstInteraction") (InteractionType :type TInteractionType%20Enum :default "IT_NoInteraction"))
  (return (Variables|Default|GetButtonCanBePressed)))

(fn BlueprintGetInteractionString
  (params (InteractingPawn :type Fort%20Pawn%20Object%20Reference) (InteractionBeingAttempted :type EInteractionBeingAttempted%20Enum :default "FirstInteraction"))
  (return (Variables|Default|GetButtonInteractionText)))

(fn BlueprintGetInteractionTime
  (params (InteractingPawn :type Fort%20Pawn%20Object%20Reference) (InteractionBeingAttempted :type EInteractionBeingAttempted%20Enum :default "FirstInteraction"))
  (return (Variables|Default|GetHackDuration) true))

(fn OnRep_ButtonIsPressed)

(fn OnRep_HackingInProgress
  (Networking|SwitchHasAuthority
    (:Remote
      (if (Variables|Default|GetHackingInProgress)
        (Audio|Components|Audio|Play (Variables|Default|GetHackingLoopingAudio) 0.0)
        (else
          (Audio|Components|Audio|FadeOut (Variables|Default|GetHackingLoopingAudio) 0.5 0.0 "Linear" :!class /Script/Engine.AudioComponent))))))

(fn ConstructionScript)
