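; Prison cell vault door Blueprint that players open by interacting while grid power is off, optionally consuming a keycard.
; A timeline rotates the door mesh open, with sounds played on non-authority instances.
; Restored power starts a timer that shuts an open door, sounding a brief alarm unless it was hacked; power loss cancels the timer.
; A hack event forces the door open, and closing fully re-enables interaction.
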(event EventBeginPlay
  (bind _doormesh (Variables|Default|GetDoorMesh))
  (bind _relativerotation (Variables|/Script/Engine.SceneComponent|GetRelativeRotation _doormesh))
  ; Init
  (Variables|BuildingActor|SetbIsInvulnerable true)
  (Variables|Default|SetStartRelativeRotation _relativerotation)
  (Variables|Default|SetEndRelativeRotation (Math|Rotator|CombineRotators _relativerotation "0.000000,-85.000000,0.000000"))
  (Variables|Default|SetIsCurrentDoorOpen false)
  (Variables|Default|SetRequireKeyCard (ScalableFloat|GetValueasBool (Variables|Default|GetRequireKeyCardHF))))

(event Interaction|EventBlueprintOnInteract
  (params (InteractingPawn :type Fort%20Pawn%20Object%20Reference) (InteractionBeingAttempted :type EInteractionBeingAttempted%20Enum :default "FirstInteraction"))
  ; Interact
  (Variables|Default|SetCurrentPlayerOpeningDoor InteractingPawn)
  (Variables|Default|SetDoorIsHacked false)
  (Variables|Default|SetCanOpenVault false)
  (CallFunction|OpenCellDoor)
  (CallFunction|RemoveKeyonUse))

(event Custom|CellHackForceDoorOpen
  ; Interact
  (if (not (Variables|Default|GetIsCurrentDoorOpen))
    (Variables|Default|SetDoorIsHacked true)
    (Variables|Default|SetCanOpenVault false)
    (CallFunction|OpenCellDoor)
    (CallFunction|RemoveKeyonUse)))

(event Custom|UpdatePowerGridState
  (params (Power :type Boolean))
  ; UpdatedCurrentPowerState
  (bind _output_get (Variables|Default|SetCurrentPowerGridState Power))
  (if _output_get
    (if (Variables|Default|GetIsCurrentDoorOpen)
      (bind _returnvalue (Utilities|Time|SetTimerbyEvent (EventDispatchers|CreateEvent :!event Custom|ShutCellDoor) (Variables|Default|GetShutDoorIfPowerChangedDelay)))
      (Variables|Default|SetForceShutCellDoorTimer _returnvalue)
      (Variables|Default|GetDoorIsHacked
        (:then)
        (:else
          (CallFunction|PlayAlarmSound)
          (CallFunction|StopAlarmSound 3.0))))
    (else
      (Utilities|Time|ClearandInvalidateTimerbyHandle (Variables|Default|GetForceShutCellDoorTimer)))))

(event Custom|OpenCellDoor
  ; Door Anims
  (CallFunction|OpenDoorAudio)
  (Utilities|FlowControl|Delay 0.3) ; to line up audio
  (bind (_direction _lerp) (Timeline|CellDoor Play)
    (:Update
      (Transformation|SetRelativeRotation (Variables|Default|GetDoorMesh) (Math|Rotator|Lerp(Rotator) (Variables|Default|GetStartRelativeRotation) (Variables|Default|GetEndRelativeRotation) _lerp true) false false))
    (:Finished
      (switch Utilities|FlowControl|Switch|SwitchonETimelineDirection _direction
        (:Forward
          (Variables|Default|SetIsCurrentDoorOpen true))
        (:Backward
          (Variables|Default|SetIsCurrentDoorOpen false)
          (Variables|Default|SetCanOpenVault true))))))

(event Custom|ShutCellDoor
  ; Door Anims
  (CallFunction|ForceShutDoorAudio)
  (Timeline|CellDoor Reverse))

(event Custom|ForceShutDoorAudio
  ; Door Audio
  (Networking|SwitchHasAuthority
    (:Remote
      (Audio|PlaySoundatLocation (Variables|Sounds|GetDoor%20Force%20Close%20Sound) (Transformation|GetWorldLocation (Variables|Default|GetDefaultSceneRoot))))))

(event Custom|OpenDoorAudio
  ; Door Audio
  (Networking|SwitchHasAuthority
    (:Remote
      (Audio|PlaySoundatLocation (Variables|Sounds|GetDoor%20Open%20Sound) (Transformation|GetWorldLocation (Variables|Default|GetDefaultSceneRoot))))))

(event Custom|RemoveKeyOnUse
  ; Remove Key on Use
  (if (Variables|Default|GetRequireKeyCard)
    (bind _asfort_world_item_definition (Utilities|Casting|CastToFortWorldItemDefinition (Variables|Default|GetKeyItemDefinition)))
    (Inventory|RemoveItemfromInventoryOwner (Fortnite|GetFortPlayerControllerfromActor (Variables|Default|GetCurrentPlayerOpeningDoor)) _asfort_world_item_definition)))

(event Custom|PlayAlarmSound
  ; Alarm Loop
  (Audio|Components|Audio|Play (Variables|Default|GetAlarmSound) 0.0))

(event Custom|StopAlarmSound
  (params (Delay :type Float%20%28double-precision%29) (FadeOutDuration :type Float%20%28double-precision%29))
  ; Alarm Loop
  (Utilities|FlowControl|Delay Delay)
  (Audio|Components|Audio|FadeOut (Variables|Default|GetAlarmSound) FadeOutDuration 0.0 "Linear" :!class /Script/Engine.AudioComponent))

(fn BlueprintCanInteract
  (params (InteractingPawn :type Fort%20Pawn%20Object%20Reference) (InteractionBeingAttempted :type EInteractionBeingAttempted%20Enum :default "FirstInteraction") (InteractionType :type TInteractionType%20Enum :default "IT_NoInteraction"))
  (return (and (and (!= (Variables|Default|GetCurrentPowerGridState) true) (Variables|Default|GetCanOpenVault)) (select (Variables|Default|GetRequireKeyCard) (>= (Inventory|GetItemQuantityonInventoryOwner (Fortnite|GetFortPlayerControllerfromActor InteractingPawn) (Variables|Default|GetKeyItemDefinition)) 1) true))))

(fn BlueprintGetFailedInteractionString
  (params (InteractingPawn :type Fort%20Pawn%20Object%20Reference) (InteractionBeingAttempted :type EInteractionBeingAttempted%20Enum :default "FirstInteraction"))
  (if (Variables|Default|GetRequireKeyCard)
    (if (and (!= (Variables|Default|GetCurrentPowerGridState) true) (Variables|Default|GetCanOpenVault))
      (return (Variables|Default|GetMissingKeyCardText))
      (else
        (return "")))
    (else
      (return ""))))

(fn BlueprintGetInteractionString
  (params (InteractingPawn :type Fort%20Pawn%20Object%20Reference) (InteractionBeingAttempted :type EInteractionBeingAttempted%20Enum :default "FirstInteraction"))
  (return "NSLOCTEXT(\"[927C57CBC117D8895A589C09A21982E9]\", \"D238229C4DB4EF9916A4E19A87DD731C\", \"Open Prison Cell\")"))

(fn BlueprintGetInteractionTime
  (params (InteractingPawn :type Fort%20Pawn%20Object%20Reference) (InteractionBeingAttempted :type EInteractionBeingAttempted%20Enum :default "FirstInteraction"))
  (return 0.75 true))

(fn ConstructionScript)
