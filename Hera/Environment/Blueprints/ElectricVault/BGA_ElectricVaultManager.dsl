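; Manages an electric vault whose powered field launches and damages players who touch it, unless a hotfix flag disables the vault.
; Either grid warning light keeps power on, updating the door and cell hack actor.
; An inside hack cuts power and forces the door open for a timed window, caching and later restoring button states.
; OnRep functions drive field collision, effects, materials, sounds and the remote alarm.
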(event EventEndPlay
  (params (EndPlayReason :type EEndPlayReason%20Enum :default "Destroyed"))
  ; End Play
  (Class|BGAElectricDoorVault|StopAlarmSound (Variables|Default|GetDoorRef) 0.0 0.0))

(event EventBeginPlay
  ; Init
  (Variables|BuildingActor|SetbIsInvulnerable true)
  (if (ScalableFloat|GetValueasBool (Variables|Hotfixable|GetEnableVault))
    (Variables|Default|SetPowerIsEnabled true)
    (Class|BGAElectricDoorVault|UpdatePowerGridState (Variables|Default|GetDoorRef) true)
    (Class|BGACellPowerShutdown|SetPowerState (Variables|Default|GetCellPowerHackRef) true)
    (else
      (Variables|Default|SetPowerIsEnabled false)
      (Utilities|FlowControl|Delay 0.25)
      (CallFunction|DestroyAllSupportingActors))))

(event OnComponentBeginOverlap(ElectricOverlapBoxFront)
  (params (OverlappedComponent :type Primitive%20Component%20Object%20Reference) (OtherActor :type Actor%20Object%20Reference) (OtherComp :type Primitive%20Component%20Object%20Reference) (OtherBodyIndex :type Integer) (bFromSweep :type Boolean) (SweepResult :type Hit%20Result%20Structure%20%28by%20ref%29))
  (bind _currentpawn (Variables|Default|GetCurrentPawn))
  (bind _returnvalue (Transformation|GetActorLocation _currentpawn))
  (bind _returnvalue_2 (+ _returnvalue "0.000000,0.000000,32.000000"))
  (bind _currentpawn_2 (Variables|Default|GetCurrentPawn))
  ; Electric Field Pushback
  (bind _asfort_player_pawn_athena (Utilities|Casting|CastToFortPlayerPawnAthena OtherActor))
  (Variables|Default|SetCurrentPawn _asfort_player_pawn_athena)
  (if (Variables|Default|GetPowerIsEnabled)
    (Variables|Default|SetImpulseDirection true)
    (Networking|SwitchHasAuthority
      (:Authority
        (if (Variables|Default|GetImpulseDirection)
          (Variables|Default|SetCurrentImpulse (+ (* (Transformation|GetForwardVector (Variables|Default|GetForwardArrow)) "500.000000,500.000000,500.000000") "0.000000,0.000000,850.000000"))
          (else
            (Variables|Default|SetCurrentImpulse (+ (* (Transformation|GetForwardVector (Variables|Default|GetBackArrow)) "500.000000,500.000000,500.000000") "0.000000,0.000000,850.000000"))))
        (Character|LaunchCharacter _currentpawn_2 (Variables|Default|GetCurrentImpulse) true true)
        (Environment|ApplyGlobalEnvironmentDamagetoActor _currentpawn_2 1.0))
      (:Remote
        (Niagara|SpawnSystematLocation "/FNGameplayCues/Effects/Fort_Effects/Effects/Characters/Generic/RankedProgression/Electric_Emote/NS_Emote_Electric_Burst.NS_Emote_Electric_Burst" _returnvalue_2)
        (Variables|Default|GetSpawnedStunnedSound
          (:then)
          (:else
            (bind _returnvalue_3 (Audio|SpawnSoundatLocation "/Game/Sounds/Fort_Traps/Wall_Electric/WallTrap_Electric_Fire_001.WallTrap_Electric_Fire_001" _returnvalue_2))
            (Variables|Default|SetSpawnedStunnedSound _returnvalue_3))) ; avoids double audio call
      ))))

(event OnComponentEndOverlap(ElectricOverlapBoxFront)
  (params (OverlappedComponent :type Primitive%20Component%20Object%20Reference) (OtherActor :type Actor%20Object%20Reference) (OtherComp :type Primitive%20Component%20Object%20Reference) (OtherBodyIndex :type Integer)))

(event OnComponentBeginOverlap(ElectricOverlapBoxBack)
  (params (OverlappedComponent :type Primitive%20Component%20Object%20Reference) (OtherActor :type Actor%20Object%20Reference) (OtherComp :type Primitive%20Component%20Object%20Reference) (OtherBodyIndex :type Integer) (bFromSweep :type Boolean) (SweepResult :type Hit%20Result%20Structure%20%28by%20ref%29))
  (bind _currentpawn (Variables|Default|GetCurrentPawn))
  (bind _returnvalue (Transformation|GetActorLocation _currentpawn))
  (bind _returnvalue_2 (+ _returnvalue "0.000000,0.000000,32.000000"))
  (bind _currentpawn_2 (Variables|Default|GetCurrentPawn))
  ; Electric Field Pushback
  (bind _asfort_player_pawn_athena (Utilities|Casting|CastToFortPlayerPawnAthena OtherActor))
  (Variables|Default|SetCurrentPawn _asfort_player_pawn_athena)
  (if (Variables|Default|GetPowerIsEnabled)
    (Variables|Default|SetImpulseDirection false)
    (Networking|SwitchHasAuthority
      (:Authority
        (if (Variables|Default|GetImpulseDirection)
          (Variables|Default|SetCurrentImpulse (+ (* (Transformation|GetForwardVector (Variables|Default|GetForwardArrow)) "500.000000,500.000000,500.000000") "0.000000,0.000000,850.000000"))
          (else
            (Variables|Default|SetCurrentImpulse (+ (* (Transformation|GetForwardVector (Variables|Default|GetBackArrow)) "500.000000,500.000000,500.000000") "0.000000,0.000000,850.000000"))))
        (Character|LaunchCharacter _currentpawn_2 (Variables|Default|GetCurrentImpulse) true true)
        (Environment|ApplyGlobalEnvironmentDamagetoActor _currentpawn_2 1.0))
      (:Remote
        (Niagara|SpawnSystematLocation "/FNGameplayCues/Effects/Fort_Effects/Effects/Characters/Generic/RankedProgression/Electric_Emote/NS_Emote_Electric_Burst.NS_Emote_Electric_Burst" _returnvalue_2)
        (Variables|Default|GetSpawnedStunnedSound
          (:then)
          (:else
            (bind _returnvalue_3 (Audio|SpawnSoundatLocation "/Game/Sounds/Fort_Traps/Wall_Electric/WallTrap_Electric_Fire_001.WallTrap_Electric_Fire_001" _returnvalue_2))
            (Variables|Default|SetSpawnedStunnedSound _returnvalue_3))) ; avoids double audio call
      ))))

(event OnComponentEndOverlap(ElectricOverlapBoxBack)
  (params (OverlappedComponent :type Primitive%20Component%20Object%20Reference) (OtherActor :type Actor%20Object%20Reference) (OtherComp :type Primitive%20Component%20Object%20Reference) (OtherBodyIndex :type Integer)))

(event Custom|PowerGridStateChanged
  ; Power State Change
  (if (Variables|Default|GetCellIsHacked)
    (Development|PrintString "Cell is hacked " (Variables|Default|GetPrintDebug) (Variables|Default|GetPrintDebug))
    (elif (Variables|/Hera/Environment/Blueprints/ElectricVault/BGA_GridWarningLight.BGA_GridWarningLight_C|GetLightOneState (Variables|Default|GetGridLightRef))
      (Development|PrintString "power on" (Variables|Default|GetPrintDebug) (Variables|Default|GetPrintDebug) :!label _label)
      (Variables|Default|SetPowerIsEnabled true)
      (Class|BGAElectricDoorVault|UpdatePowerGridState (Variables|Default|GetDoorRef) true)
      (Class|BGACellPowerShutdown|SetPowerState (Variables|Default|GetCellPowerHackRef) true)
      (if (not (Variables|Default|GetCellIsHacked))
        (Class|BGAGridWarningLight|RefreshLightStates (Variables|Default|GetGridLightRef)))
      (elif (Variables|/Hera/Environment/Blueprints/ElectricVault/BGA_GridWarningLight.BGA_GridWarningLight_C|GetLightTwoState (Variables|Default|GetGridLightRef))
        (goto _label))))
  (Development|PrintString "power off" (Variables|Default|GetPrintDebug) (Variables|Default|GetPrintDebug))
  (Variables|Default|SetPowerIsEnabled false)
  (Class|BGAElectricDoorVault|UpdatePowerGridState (Variables|Default|GetDoorRef) false)
  (Class|BGACellPowerShutdown|SetPowerState (Variables|Default|GetCellPowerHackRef) false)
  (if (not (Variables|Default|GetCellIsHacked))
    (Class|BGAGridWarningLight|RefreshLightStates (Variables|Default|GetGridLightRef))))

(event Custom|InsideCellHackInitiated
  ; Hack the cell open
  (Variables|Default|SetCellIsHacked true)
  (bind _returnvalue (Utilities|Time|SetTimerbyEvent (delegate Custom|HackTimeoutCompleted) 15.0))
  (Variables|Default|SetInsideCellHackHandle _returnvalue)
  (Class|BGAGridWarningLight|LightStateCellHacked (Variables|Default|GetGridLightRef) 0 false)
  (Class|BGAGridWarningLight|LightStateCellHacked (Variables|Default|GetGridLightRef) 1 false)
  (CallFunction|PowerGridStateChanged)
  (Class|BGAElectricDoorVault|CellHackForceDoorOpen (Variables|Default|GetDoorRef))
  (CallFunction|HandlePowerGridButtonStatewhenHacked))

(event Custom|HackTimeoutCompleted
  ; Hack the cell open
  (Variables|Default|SetCellIsHacked false)
  (Utilities|FlowControl|Delay 1.0)
  (CallFunction|PowerGridStateChanged)
  (CallFunction|RestorePowerGridButtonStateAfterHack))

(fn DestroyAllSupportingActors
  (Actor|DestroyActor (Variables|Default|GetDoorRef))
  (Actor|DestroyActor (Variables|Default|GetGridLightRef))
  (for _array_element (Variables|Default|GetPowerGridButtons)
    (Actor|DestroyActor _array_element))
  (for _array_element_2 (Variables|Default|GetAllVaultContainers)
    (Gameplay|SilentDie _array_element_2 false))
  (Actor|DestroyActor (Variables|Default|GetHackStationRef)))

(fn HandlePowerGridButtonStateWhenHacked
  (Utilities|Array|Clear (Variables|Default|GetCachedPowerGridButtonStatePreHack))
  (for (_array_element _array_index) (Variables|Default|GetPowerGridButtons)
    (Utilities|Array|SetArrayElem (Variables|Default|GetCachedPowerGridButtonStatePreHack) _array_index (Variables|/Hera/Environment/Blueprints/ElectricVault/BGA_PowerGridButton.BGA_PowerGridButton_C|GetCurrentPowerState _array_element) true)
    (Class|BGAPowerGridButton|ForcePowerState _array_element false)))

(fn OnRep_CellIsHacked
  (Networking|SwitchHasAuthority
    (:Remote
      (if (Variables|Default|GetCellIsHacked)
        (Class|BGAElectricDoorVault|PlayAlarmSound (Variables|Default|GetDoorRef))
        (else
          (Class|BGAElectricDoorVault|StopAlarmSound (Variables|Default|GetDoorRef) 2.0 1.0))))))

(fn OnRep_PowerIsEnabled
  (bind _electricityloop (Variables|Default|GetElectricityLoop))
  (bind _electricityloop_2 (Variables|Default|GetElectricityLoop))
  (if (Variables|Default|GetPowerIsEnabled)
    (Components|Activation|Activate (Variables|Default|GetNS_ElectricVault_Loop) true)
    (if (not (Audio|Components|Audio|IsPlaying _electricityloop))
      (Audio|Components|Audio|FadeIn _electricityloop 0.5 1.0 0.0 "Linear" :!class /Script/Engine.AudioComponent))
    (Collision|SetCollisionEnabled (Variables|Default|GetCollisionBox) "QueryAndPhysics")
    (Rendering|Material|SetMaterial (Variables|Default|GetSM_FriedNexus_Pillar_Base_A_A1) 0 "/Hera/Sets/DarkNexus/Materials/MI_DarkNexus_Pillar_A_VaultClosed.MI_DarkNexus_Pillar_A_VaultClosed")
    (Rendering|Material|SetMaterial (Variables|Default|GetSM_FloatMax_CellGate_A_Center) 0 "/Game/Environments/Helios/Sets/FloatMax/Materials/MI_FloatMax_CellGate_A_A_VaultClosed.MI_FloatMax_CellGate_A_A_VaultClosed")
    (Audio|PlaySoundatLocation (Variables|Sounds|GetPower%20Activate%20Sound) (Transformation|GetWorldLocation (Variables|Default|GetDefaultSceneRoot)) "0, 0, 0" 1.0 0.9)
    (else
      (Components|Activation|Deactivate (Variables|Default|GetNS_ElectricVault_Loop))
      (if (Audio|Components|Audio|IsPlaying _electricityloop_2)
        (Audio|Components|Audio|FadeOut _electricityloop_2 1.0 0.0 "Linear" :!class /Script/Engine.AudioComponent))
      (Collision|SetCollisionEnabled (Variables|Default|GetCollisionBox) "NoCollision")
      (Rendering|Material|SetMaterial (Variables|Default|GetSM_FriedNexus_Pillar_Base_A_A1) 0 "/Hera/Sets/DarkNexus/Materials/MI_DarkNexus_Pillar_A_Inst_VaultOpen.MI_DarkNexus_Pillar_A_Inst_VaultOpen")
      (Rendering|Material|SetMaterial (Variables|Default|GetSM_FloatMax_CellGate_A_Center) 0 "/Game/Environments/Helios/Sets/FloatMax/Materials/MI_FloatMax_CellGate_A_A_VaultOpen.MI_FloatMax_CellGate_A_A_VaultOpen")
      (Audio|PlaySoundatLocation (Variables|Sounds|GetPower%20Deactivate%20Sound) (Transformation|GetWorldLocation (Variables|Default|GetDefaultSceneRoot)) "0, 0, 0" 1.0 0.9))))

(fn RestorePowerGridButtonStateAfterHack
  (for (_array_element _array_index) (Variables|Default|GetPowerGridButtons)
    (Class|BGAPowerGridButton|ForcePowerState _array_element (Utilities|Array|Get(acopy) (Variables|Default|GetCachedPowerGridButtonStatePreHack) _array_index))))

(fn ConstructionScript)
