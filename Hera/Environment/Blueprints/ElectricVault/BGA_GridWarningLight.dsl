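; Invulnerable two-light warning fixture for the Electric Vault grid, showing each light as red (enabled) or green (disabled).
; LightStateToggle sets index and state, whose OnRep swaps material and color and records per-light state, then plays on/off audio on remotes, using previous-state tracking against repeats.
; LightStateCellHacked changes color through a separate OnRep without sound or touching recorded state.
; RefreshLightStates reapplies the recorded states through LightStateToggle.
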(event EventBeginPlay
  (Variables|BuildingActor|SetbIsInvulnerable true)
  ; This allows the first interaction of the lights to play sound
  (Variables|Default|SetPreviousLightOneState true)
  (Variables|Default|SetPreviousLightTwoState true))

(event Custom|LightStateToggle
  (params (LightIndex :type Integer) (Enabled :type Boolean))
  (Variables|Default|SetCurrentLightIndex LightIndex)
  (Variables|Default|SetLightState Enabled)
  (Variables|Default|SetLightStateLocal Enabled)
  (CallFunction|PlaySound))

(event Custom|LightStateCellHacked
  (params (LightIndex :type Integer) (Enabled :type Boolean))
  (Variables|Default|SetCurrentLightIndex LightIndex)
  (Variables|Default|SetLightStateHacked Enabled)
  (Variables|Default|SetLightStateLocal Enabled))

(event Custom|RefreshLightStates
  (CallFunction|LightStateToggle 0 (Variables|Default|GetLightOneState))
  (CallFunction|LightStateToggle 1 (Variables|Default|GetLightTwoState)))

(event Custom|PlaySound
  (Networking|SwitchHasAuthority
    (:Remote
      (switch int (Variables|Default|GetCurrentLightIndex)
        (:0
          ; This helps us avoid multiple audio calls. We check to see if the light state has changed, if it has, we continue playing audio. an activate/deactive sound is played depending on the local light state
          (if (!= (Variables|Default|GetLightOneState) (Variables|Default|GetPreviousLightOneState))
            _))
        (:1
          (if (!= (Variables|Default|GetLightTwoState) (Variables|Default|GetPreviousLightTwoState))
            _)))
      (Audio|SpawnSoundatLocation (select (Variables|Default|GetLightStateLocal) (Variables|Sounds|GetLight%20Toggle%20Off%20Sound) (Variables|Sounds|GetLight%20Toggle%20On%20Sound)) (Transformation|GetWorldLocation (Variables|Default|GetAudioLocation))) ; light off
      (Variables|Default|SetPreviousLightOneState (Variables|Default|GetLightOneState))
      (Variables|Default|SetPreviousLightTwoState (Variables|Default|GetLightTwoState)))))

(fn OnRep_LightState
  (if (Variables|Default|GetLightState)
    (switch int (Variables|Default|GetCurrentLightIndex)
      (:0
        (Rendering|Material|SetMaterial (Variables|Default|GetSM_Light_0) 0 "/Game/Environments/Helios/Props/RustBucket/RustBucket_Sconces_A/Materials/MI_RustBucket_Sconces_A_Red.MI_RustBucket_Sconces_A_Red")
        (Variables|Default|SetLightOneState true)
        (Rendering|Components|Light|SetLightColor (Variables|Default|GetPointLight_0) "(R=1.000000,G=0.000000,B=0.000000,A=1.000000)" true))
      (:1
        (Rendering|Material|SetMaterial (Variables|Default|GetSM_Light_1) 0 "/Game/Environments/Helios/Props/RustBucket/RustBucket_Sconces_A/Materials/MI_RustBucket_Sconces_A_Red.MI_RustBucket_Sconces_A_Red")
        (Variables|Default|SetLightTwoState true)
        (Rendering|Components|Light|SetLightColor (Variables|Default|GetPointLight_1) "(R=1.000000,G=0.000000,B=0.000000,A=1.000000)" true)))
    (else
      (switch int (Variables|Default|GetCurrentLightIndex)
        (:0
          (Rendering|Material|SetMaterial (Variables|Default|GetSM_Light_0) 0 "/Game/Environments/Helios/Props/RustBucket/RustBucket_Sconces_A/Materials/MI_RustBucket_Sconces_A_Green.MI_RustBucket_Sconces_A_Green")
          (Variables|Default|SetLightOneState false)
          (Rendering|Components|Light|SetLightColor (Variables|Default|GetPointLight_0) "(R=0.000000,G=1.000000,B=0.000000,A=1.000000)" true))
        (:1
          (Rendering|Material|SetMaterial (Variables|Default|GetSM_Light_1) 0 "/Game/Environments/Helios/Props/RustBucket/RustBucket_Sconces_A/Materials/MI_RustBucket_Sconces_A_Green.MI_RustBucket_Sconces_A_Green")
          (Variables|Default|SetLightTwoState false)
          (Rendering|Components|Light|SetLightColor (Variables|Default|GetPointLight_1) "(R=0.000000,G=1.000000,B=0.000000,A=1.000000)" true))))))

(fn OnRep_LightStateHacked
  (if (Variables|Default|GetLightStateHacked)
    (switch int (Variables|Default|GetCurrentLightIndex)
      (:0
        (Rendering|Material|SetMaterial (Variables|Default|GetSM_Light_0) 0 "/Game/Environments/Helios/Props/RustBucket/RustBucket_Sconces_A/Materials/MI_RustBucket_Sconces_A_Red.MI_RustBucket_Sconces_A_Red")
        (Rendering|Components|Light|SetLightColor (Variables|Default|GetPointLight_0) "(R=1.000000,G=0.000000,B=0.000000,A=1.000000)" true))
      (:1
        (Rendering|Material|SetMaterial (Variables|Default|GetSM_Light_1) 0 "/Game/Environments/Helios/Props/RustBucket/RustBucket_Sconces_A/Materials/MI_RustBucket_Sconces_A_Red.MI_RustBucket_Sconces_A_Red")
        (Rendering|Components|Light|SetLightColor (Variables|Default|GetPointLight_1) "(R=1.000000,G=0.000000,B=0.000000,A=1.000000)" true)))
    (else
      (switch int (Variables|Default|GetCurrentLightIndex)
        (:0
          (Rendering|Material|SetMaterial (Variables|Default|GetSM_Light_0) 0 "/Game/Environments/Helios/Props/RustBucket/RustBucket_Sconces_A/Materials/MI_RustBucket_Sconces_A_Green.MI_RustBucket_Sconces_A_Green")
          (Rendering|Components|Light|SetLightColor (Variables|Default|GetPointLight_0) "(R=0.000000,G=1.000000,B=0.000000,A=1.000000)" true))
        (:1
          (Rendering|Material|SetMaterial (Variables|Default|GetSM_Light_1) 0 "/Game/Environments/Helios/Props/RustBucket/RustBucket_Sconces_A/Materials/MI_RustBucket_Sconces_A_Green.MI_RustBucket_Sconces_A_Green")
          (Rendering|Components|Light|SetLightColor (Variables|Default|GetPointLight_1) "(R=0.000000,G=1.000000,B=0.000000,A=1.000000)" true))))))

(fn ConstructionScript)
