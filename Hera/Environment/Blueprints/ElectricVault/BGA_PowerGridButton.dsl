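; An invulnerable interactable button in the electric vault that toggles a location's prison security grid on or off.
; Interaction or ForcePowerState flips the power state, updates prompt and warning light, then notifies the vault manager, which blocks interaction once its cell is hacked.
; When a keycard is required, it spawns one unless the player holds one or a matching pickup is nearby.
; Clients get sounds, with red/green lighting driven by the power state's RepNotify.
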(event EventBeginPlay
  ; Init
  (Variables|Default|SetButtonInteractionText (Utilities|Text|FormatText "NSLOCTEXT(\"[2223AC04A7FDDB47FF2208E9EFCA7BC1]\", \"1B828ACC4D4EF10A6FED638B1FBEDB0F\", \"Shutdown {y} Prison Security Grid\")" (Variables|Default|GetLocation)))
  (Variables|BuildingActor|SetbIsInvulnerable true)
  (Variables|Default|SetRequireKeyCard (ScalableFloat|GetValueasBool (Variables|Default|GetRequireKeyCardHF))))

(event Custom|ForcePowerState
  (params (State :type Boolean))
  ; Button Interaction Logic
  (if State
    (Variables|Default|SetCurrentPowerState true)
    (Class|BGAGridWarningLight|LightStateToggle (Variables|Default|GetGridWarningLightRef) (Variables|Default|GetGridNumber) true)
    (Variables|Default|SetButtonInteractionText (Utilities|Text|FormatText "NSLOCTEXT(\"[2223AC04A7FDDB47FF2208E9EFCA7BC1]\", \"AD5E686B424979D342AAF5A72136772C\", \"Shutdown {y} Prison Security Grid\")" (Variables|Default|GetLocation)))
    (CallFunction|ChangedPowerState true)
    (else
      (Variables|Default|SetCurrentPowerState false)
      (Class|BGAGridWarningLight|LightStateToggle (Variables|Default|GetGridWarningLightRef) (Variables|Default|GetGridNumber) false)
      (Variables|Default|SetButtonInteractionText (Utilities|Text|FormatText "NSLOCTEXT(\"[2223AC04A7FDDB47FF2208E9EFCA7BC1]\", \"3585CD444326DF6DDEE8D887C22156BF\", \"Enable {y} Prison Security Grid\")" (Variables|Default|GetLocation)))
      (CallFunction|ChangedPowerState false)))
  (Utilities|FlowControl|Delay 0.33)
  (CallFunction|SpawnKeyItem)
  (Class|BGAElectricVaultManager|PowerGridStateChanged (Variables|Default|GetVaultManagerRef))
  (Utilities|FlowControl|Delay 1.0)
  (Variables|Default|SetButtonCanBePressed true))

(event Interaction|EventBlueprintOnInteract
  (params (InteractingPawn :type Fort%20Pawn%20Object%20Reference) (InteractionBeingAttempted :type EInteractionBeingAttempted%20Enum :default "FirstInteraction"))
  ; Button Interaction Logic
  (Variables|Default|SetCurrentInteractingPlayer InteractingPawn)
  (Variables|Default|SetButtonCanBePressed false)
  (if (Variables|Default|GetCurrentPowerState)
    (Variables|Default|SetCurrentPowerState false)
    (Class|BGAGridWarningLight|LightStateToggle (Variables|Default|GetGridWarningLightRef) (Variables|Default|GetGridNumber) false)
    (Variables|Default|SetButtonInteractionText (Utilities|Text|FormatText "NSLOCTEXT(\"[2223AC04A7FDDB47FF2208E9EFCA7BC1]\", \"3585CD444326DF6DDEE8D887C22156BF\", \"Enable {y} Prison Security Grid\")" (Variables|Default|GetLocation)))
    (CallFunction|ChangedPowerState false)
    (else
      (Variables|Default|SetCurrentPowerState true)
      (Class|BGAGridWarningLight|LightStateToggle (Variables|Default|GetGridWarningLightRef) (Variables|Default|GetGridNumber) true)
      (Variables|Default|SetButtonInteractionText (Utilities|Text|FormatText "NSLOCTEXT(\"[2223AC04A7FDDB47FF2208E9EFCA7BC1]\", \"AD5E686B424979D342AAF5A72136772C\", \"Shutdown {y} Prison Security Grid\")" (Variables|Default|GetLocation)))
      (CallFunction|ChangedPowerState true)))
  (Utilities|FlowControl|Delay 0.33)
  (CallFunction|SpawnKeyItem)
  (Class|BGAElectricVaultManager|PowerGridStateChanged (Variables|Default|GetVaultManagerRef))
  (Utilities|FlowControl|Delay 1.0)
  (Variables|Default|SetButtonCanBePressed true))

(event Custom|ButtonPressedEvent
  ; Button Interaction Event - What Happens?
  ;
  (Networking|SwitchHasAuthority
    (:Authority
      (Variables|Default|SetButtonIsPressed true) ; this changes the lights visibility on the clients
    )
    (:Remote
      ;;[0.15 0.01327 0 0.5] NOT HOOKED UP
      (Audio|Components|Audio|Play (Variables|Default|GetButtonPressedAudio) 0.0) ; play the audio component  button press sound
      (Niagara|SpawnSystematLocation "/Game/Athena/Artemis/Environments/BuildingActors/Daisy/FX/NS_Daisy_Mushroom_Bounce.NS_Daisy_Mushroom_Bounce" (Transformation|GetWorldLocation (Variables|Default|GetArrow))) ; Spawn a new niagara particle effect at this location
    )) ; authority = server, Remote = Client
)

(event Custom|SpawnKeyItem
  ; Spawn Key
  (if (Variables|Default|GetRequireKeyCard)
    (bind _keyfound (CallFunction|CheckKeyinProximity))
    (if (not _keyfound)
      (bind _isininventory (CallFunction|CheckKeyAlreadyinInventory))
      (if (not _isininventory)
        (CallFunction|GrantKeytoPlayer)
        (CallFunction|KeyGrantedAudio)))))

(event Custom|ChangedPowerState
  (params (Active :type Boolean))
  ; Client FX
  (Networking|SwitchHasAuthority
    (:Remote
      (if Active
        (Audio|SpawnSoundatLocation (Variables|Sounds|GetDeactivate%20Sound) (Transformation|GetWorldLocation (Variables|Default|GetButtonPressedAudio)))
        (else
          (Audio|SpawnSoundatLocation (Variables|Sounds|GetActivate%20Sound) (Transformation|GetWorldLocation (Variables|Default|GetButtonPressedAudio))))))))

(event Custom|KeyGrantedAudio
  ; Client FX
  (Networking|SwitchHasAuthority
    (:Remote
      (Audio|SpawnSoundatLocation :Sound "/Game/Athena/Sounds/LTMs/Bling/Bling_Llama_Drop_Cue.Bling_Llama_Drop_Cue" :Location (Transformation|GetWorldLocation (Variables|Default|GetDefaultSceneRoot)) :VolumeMultiplier 1.0 :PitchMultiplier 1.0 :StartTime 0.0 :AttenuationSettings "/Game/Athena/Sounds/Events/Papaya/SpeakerSystem/Papaya_Speaker_Medium_Att.Papaya_Speaker_Medium_Att" :bAutoDestroy true))))

(fn BlueprintCanInteract
  (params (InteractingPawn :type Fort%20Pawn%20Object%20Reference) (InteractionBeingAttempted :type EInteractionBeingAttempted%20Enum :default "FirstInteraction") (InteractionType :type TInteractionType%20Enum :default "IT_NoInteraction"))
  (return (and (not (Variables|/Hera/Environment/Blueprints/ElectricVault/BGA_ElectricVaultManager.BGA_ElectricVaultManager_C|GetCellIsHacked (Variables|Default|GetVaultManagerRef))) (Variables|Default|GetButtonCanBePressed))))

(fn BlueprintGetFailedInteractionString
  (params (InteractingPawn :type Fort%20Pawn%20Object%20Reference) (InteractionBeingAttempted :type EInteractionBeingAttempted%20Enum :default "FirstInteraction"))
  (if (Variables|/Hera/Environment/Blueprints/ElectricVault/BGA_ElectricVaultManager.BGA_ElectricVaultManager_C|GetCellIsHacked (Variables|Default|GetVaultManagerRef))
    (return (Variables|Default|GetHackedOpenInteractionText))
    (else
      (return ""))))

(fn BlueprintGetInteractionString
  (params (InteractingPawn :type Fort%20Pawn%20Object%20Reference) (InteractionBeingAttempted :type EInteractionBeingAttempted%20Enum :default "FirstInteraction"))
  (return (Variables|Default|GetButtonInteractionText)))

(fn CheckKeyAlreadyInInventory
  (return (>= (Inventory|GetItemQuantityonInventoryOwner (Fortnite|GetFortPlayerControllerfromActor (Variables|Default|GetCurrentInteractingPlayer)) (Variables|Default|GetKeyItemDefinition)) 1)))

(fn CheckKeyInProximity
  (bind _outactors (Collision|SphereOverlapActors (Transformation|GetWorldLocation (Variables|Default|GetDefaultSceneRoot)) 1e+03 (Utilities|Array|MakeArray "ObjectTypeQuery4") "/Script/FortniteGame.FortPickupAthena"))
  (bind _array_element (Utilities|Array|ForEachLoopwithBreak _outactors)
    (:LoopBody
      (bind _asfort_pickup_athena (Utilities|Casting|CastToFortPickupAthena _array_element))
      (if (== (Pickup|GetPrimaryItemDefinition _asfort_pickup_athena) (Variables|Default|GetKeyItemDefinition))
        (LocalVariables|SetKeyLocal true)))
    (:Completed
      _))
  (return (LocalVariables|GetKeyLocal)))

(fn GrantKeyToPlayer
  (if (Variables|Default|GetRequireKeyCard)
    (Inventory|SpawnPickupInWorld (Variables|Default|GetKeyItemDefinition) 1 (+ (+ (Transformation|GetActorLocation self) (* (Transformation|GetForwardVector (Variables|Default|GetArrow)) 75.0)) "0.000000,0.000000,50.000000") (Transformation|GetForwardVector (Variables|Default|GetArrow)) 1)))

(fn OnRep_ButtonIsPressed)

(fn OnRep_CurrentPowerState
  (if (Variables|Default|GetCurrentPowerState)
    (Rendering|Material|SetMaterial (Variables|Default|GetSM_Light_0) 0 "/Game/Environments/Helios/Props/RustBucket/RustBucket_Sconces_A/Materials/MI_RustBucket_Sconces_A_Red.MI_RustBucket_Sconces_A_Red")
    (Rendering|Components|Light|SetLightColor (Variables|Default|GetPointLight) "(R=1.000000,G=0.000000,B=0.000000,A=1.000000)" true)
    (else
      (Rendering|Material|SetMaterial (Variables|Default|GetSM_Light_0) 0 "/Game/Environments/Helios/Props/RustBucket/RustBucket_Sconces_A/Materials/MI_RustBucket_Sconces_A_Green.MI_RustBucket_Sconces_A_Green")
      (Rendering|Components|Light|SetLightColor (Variables|Default|GetPointLight) "(R=0.000000,G=1.000000,B=0.000000,A=1.000000)" true))))

(fn OnRep_LightState)

(fn ConstructionScript)
